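; Smash brute Pulse Lancer AI pawn that extends its parent's begin-play and death handling.
; On death, it sends an Athena.LTM.Mash.AI.Died gameplay event to its Instigator, carrying itself and the killing controller, to trigger Mash LTM death abilities (logic ported from AthenaAI_Fiend).
; On begin play, it sets the SM_ShieldBreaker mesh to render in custom depth only, outside the main pass.
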(event EventCustomOnDied
  (params (DamagedActor :type Actor%20Object%20Reference) (Damage :type Float%20%28single-precision%29) (InstigatedBy :type Controller%20Object%20Reference) (DamageCauser :type Actor%20Object%20Reference) (HitLocation :type Vector :default "0, 0, 0") (FHitComponent :type Primitive%20Component%20Object%20Reference) (BoneName :type Name :default "None") (Momentum :type Vector :default "0, 0, 0"))
  ;[1 0.593 0.3655] Pass an Event to Trigger On Death Abilities for Mash LTM (Functionality brought over from AthenaAI_Fiend)
  (|CallParentFunction DamagedActor Damage InstigatedBy DamageCauser HitLocation FHitComponent BoneName Momentum)
  (Ability|SendGameplayEventtoActor (Variables|Actor|GetInstigator) "(TagName=\"Athena.LTM.Mash.AI.Died\")" (Utilities|Struct|MakeGameplayEventData :Instigator self :OptionalObject InstigatedBy :ContextHandle "()" :InstigatorTags "(GameplayTags=)" :TargetTags "(GameplayTags=)" :EventMagnitude 0.0 :TargetData "()")))

(event EventBeginPlay
  (|CallParentFunction)
  (Rendering|SetRenderCustomDepth (Variables|Default|GetSM_ShieldBreaker) true)
  (Rendering|SetRenderinMainPass (Variables|Default|GetSM_ShieldBreaker) false :!class /Script/Engine.PrimitiveComponent))

(fn ConstructionScript
  (|CallParentFunction))
