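; Child Blueprint for the Smash ranged "Chill" PulseLancer pawn, layering a small render setup over its parent class.
; On BeginPlay it runs the parent logic, then enables custom depth and disables main-pass rendering on the inherited-or-local SM_ShieldBreaker component.
; That mesh thus draws to the custom depth buffer while staying out of the main pass.
; The construction script only calls the parent.
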(event EventBeginPlay
  (|CallParentFunction)
  (Rendering|SetRenderCustomDepth (Variables|Default|GetSM_ShieldBreaker) true)
  (Rendering|SetRenderinMainPass (Variables|Default|GetSM_ShieldBreaker) false :!class /Script/Engine.PrimitiveComponent))

(fn ConstructionScript
  (|CallParentFunction))
